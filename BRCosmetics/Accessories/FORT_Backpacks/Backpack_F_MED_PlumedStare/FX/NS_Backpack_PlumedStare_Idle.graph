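# Niagara idle effect for the PlumedStare backpack cosmetic, drawing teal flame meshes from one emitter.
# A burst spawns 13 persistent particles, whose partition index picks color, inner color, scale, offset, orientation and visibility tag from arrays for the mesh renderers.
# Updates feed a smoothed world position into dynamic material parameters and scale both colors by time-of-day curves and a system fade-in.
# Camera offset differs between gameplay and front end.
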NiagaraSystem /BRCosmetics/Accessories/FORT_Backpacks/Backpack_F_MED_PlumedStare/FX/NS_Backpack_PlumedStare_Idle
dependencies {
 usedDataInterfaces [
  &NiagaraDataInterfaceArrayColor
  &NiagaraDataInterfaceArrayFloat3
  &NiagaraDataInterfaceArrayInt32
  &NiagaraDataInterfaceColorCurve
  &NiagaraDataInterfaceCurve
 ]
 usedDynamicInputs [&FloatFromCurve]
 usedModules [
  &SystemState
  &self:SystemSpawnScript.SystemScriptSource.SystemScriptGraph.NiagaraNodeAssignment_0.SetVariables_#1
 ]
}
emitters {FlamesSplinesCombined {
  emitterUpdateScript [
   &EmitterState(
    Inactive Response=::NewEnumerator1
    Life Cycle Mode=::NewEnumerator1
    Loop Behavior=::NewEnumerator1
    Loop Duration Mode=::NewEnumerator1
    UseLoopDelay=false
   )
   &SpawnBurst_Instantaneous(Spawn Count=13, Use Spawn Probability=false, UseLoopCountLimit=false)
  ]
  particleSpawnScript [
   &PartitionParticles(ID to Partition:NiagaraInt32=&ReturnExecIndex(), Number Of Partitions=13)
   &InitializeParticle(
    Color=&SelectLinearColorFromArray(
     Array Sampling Mode=::NewEnumerator1
     Direct Array Index=$Particles.PartitionParticles.Partition
     Color Selection Array=DataInterface() {propertyValues {ColorData [
        <rgba 0 0.0997 0.13 0.35>
        <rgba 0 0.075 0.1 0.35>
        <rgba 0 0.0953 0.13 0.45>
        <rgba 0 0.0888 0.13 0.3>
        <rgba 0 0.14 0.1326 0.4>
        <rgba 0 0.13 0.1231 0.3>
        <rgba 0 0.13 0.1231 0.45>
        <rgba 0 0.13 0.1231 0.45>
        <rgba 0 0.0633 0.2 0.6>
        <rgba 0 0.075 0.1 0.45>
        <rgba 0 0.16 0.1516 0.3>
        <rgba 0 0.16 0.1516 0.3>
        <rgba 0 0.13 0.1231 0.3>
       ]}}
     Random Seed=?
    )
    Lifetime=0
    Mesh Scale=&SelectVectorFromArray(
     Direct Array Index=$Particles.PartitionParticles.Partition
     Vector Selection Array=DataInterface() {propertyValues {FloatData [
        <0.55 0.5 0.6>
        <0.55 0.5 0.4>
        <0.55 0.5 0.4>
        <0.55 0.5 0.4>
        <0.55 0.5 0.6>
        <0.7 0.5 0.5>
        <0.7 0.5 0.5>
        <0.8 0.5 0.4>
        <0.6 0.5 0.4>
        <0.7 0.5 0.6>
        <0.55 0.5 0.5>
        <0.3 0.5 0.5>
        <0.7 0.5 0.5>
       ]}}
    )
    Mesh Scale Mode=::NewEnumerator4
    Position Offset=&SelectVectorFromArray(
     Direct Array Index=$Particles.PartitionParticles.Partition
     Vector Selection Array=DataInterface() {propertyValues {FloatData [
        <10 -7 -4>
        <10 -7 -8>
        <-10.5 -7 -4>
        <-10.5 -7 -6.5>
        <5 -7 -5>
        <8.8 -7 -0.5>
        <7 -7 4>
        <5 -7.5 4>
        <3.5 -7.5 5>
        <-2 -7 -12>
        <0 -6 2>
        <0 -6 3>
        <6 -5 -2>
       ]}}
    )
    UsePositionOffset=true
    Write Material Random=true
   )
   &CameraOffset(Camera Offset Amount=&FN_FrontEndFloatModifier(Gameplay=3, Front End=5))
   &InitialMeshOrientation(
    Mesh Axis To Orient=&SelectVectorFromArray(
     Direct Array Index=$Particles.PartitionParticles.Partition
     Vector Selection Array=DataInterface() {propertyValues {FloatData [<-1 0 0>, <-1 0 0>, <1 0 0>, <1 0 0>, <-1 0 0>, <1 0 0>, <1 0 0>, <1 0 0>, <1 0 0>, <1 0 0>, <-1 0 0>, <1 0 0>, <1 0 0>]}}
    )
    Mesh Orientation Mode=::NewEnumerator2
    Orientation Coordinate Space=::Local
    Rotation=&SelectVectorFromArray(
     Direct Array Index=$Particles.PartitionParticles.Partition
     Vector Selection Array=DataInterface() {propertyValues {FloatData [
        <0 0.2 1.01>
        <0 0.2 1.01>
        <0 0.044 0>
        <0 -0.0107 0>
        <0 0.2 1.01>
        <0 0.0714 0>
        <0 -0.0381 0>
        <0 -0.0563 0>
        <0 -0.1384 0>
        <0 0.044 0>
        <0 -1.5679 1.01>
        <0.5 0 -0.3931>
        <0 0.1716 0.1>
       ]}}
    )
    Use Rotation Vector=true
   )
   /SetVariables(Particles.VisibilityTag=&SelectIntFromArray(
     Direct Array Index=$Particles.PartitionParticles.Partition
     Int Selection Array=DataInterface() {propertyValues {IntData [0,1,2,3,4,5,6,7,8,9,10,11,12]}}
    )) {isSetParameters=true}
   /SetVariables(Particles.Inner Color:LinearColor=&SelectLinearColorFromArray(
     Array Sampling Mode=::NewEnumerator1
     Direct Array Index=$Particles.PartitionParticles.Partition
     Color Selection Array=DataInterface() {propertyValues {ColorData [
        <rgba 0.1735 3 2.731 1>
        <rgba 0.1735 3 2.731 1>
        <rgba 0.1735 3 2.731 1>
        <rgba 0.1735 3 2.731 1>
        <rgba 0.6361 11 10.0136 1>
        <rgba 0.5204 9 8.193 1>
        <rgba 0.5204 9 8.193 1>
        <rgba 0.9252 16 14.5653 1>
        <rgba 0 3 2.9408 1>
        <rgba 1.2721 22 20.0272 1>
        <rgba 0 6 5.5 1>
        <rgba 0.7517 13 11.8343 1>
        <rgba 0.1156 2 1.8207 1>
       ]}}
     Random Seed=?
    )) {isSetParameters=true}
  ]
  particleUpdateScript [
   &ParticleState(Kill Particles When Lifetime Has Elapsed=false)
   &InterpolateOverTime(
    Target Vector Value:Vector3f=&TransformPosition(Position=$Particles.Position, Source Space=::Local, Destination Space=::World)
    Variable Type=::NewEnumerator2
   )
   &DynamicMaterialParameters(
    Index 0 Param 1=&RandomRangeFloat(Minimum=0, Maximum=1)
    Index 0 Param 2=&Lerp_Float(
     A=1
     B=-1
     Alpha=&RemapRange(
      Input Value To Remap=&DotProduct(Vector A=$Engine.Owner.Velocity, Vector B=&TransformVector(Vector=<0 0 1>, Destination Space=::Simulation))
      Input Min=0
      Input Max=300
      Output Min=0
      Output Max=1
     )
    )
    Index 0 Param 3=&MakeFloatFromVector(Vector=<0 0 0>, Channel=::NewEnumerator0)
    Index 1 Param 1=&MakeFloatFromVector(Vector=$StackContext.InterpolateOverTime.InterpolateOverTime_V3_.Moving Average, Channel=::NewEnumerator0)
    Index 1 Param 2=&MakeFloatFromVector(Vector=$StackContext.InterpolateOverTime.InterpolateOverTime_V3_.Moving Average, Channel=::NewEnumerator1)
    Index 1 Param 3=&MakeFloatFromVector(Vector=$StackContext.InterpolateOverTime.InterpolateOverTime_V3_.Moving Average, Channel=::NewEnumerator2)
    Module.Param0WriteEnabled=false
    Module.Param0WriteEnabled1=true
    Module.Param1WriteEnabled=false
    Module.Param1WriteEnabled1=true
    Module.Param2WriteEnabled=false
    Module.Param2WriteEnabled1=true
    Module.Param3WriteEnabled=false
    Module.Param3WriteEnabled1=true
    Write Parameter Index 1=true
   ) {module=DynamicMaterialParameters001}
   /SetVariables(Particles.Inner Color:LinearColor=&MultiplyLinearColor(
     Linear Color=$Particles.Initial.Inner Color
     Scale Factor=&Multiply_Vector4(
      A=&Vector4FromCurve(
       CurveIndex=Linked() {linkedVariable {name=NPC.FortniteNPC.TimeOfDay_0-1, type=NiagaraFloat}}
       Vector4Curve=DataInterface() {propertyValues {
         AlphaCurve {keys [<at 0 1>,<at 0.125 1>,<at 0.25 1>,<at 0.375 1>,<at 0.5 1>,<at 0.625 1>,<at 0.75 1>,<at 0.875 1>,<at 1 1>]}
         BlueCurve {keys [<at 0 0.9216>,<at 0.041 0.8963>,<at 0.083 0.8308>,<at 0.125 0.7835>,<at 0.166 0.6172>,<at 0.208 0.5395>,<at 0.25 0.6172>,<at 0.291 0.6939>,<at 0.333 0.7379>,<at 0.375 0.7835>,<at 0.416 0.7991>,<at 0.458 0.8148>,<at 0.5 0.8308>,<at 0.541 0.8148>,<at 0.583 0.7991>,<at 0.625 0.7991>,<at 0.666 0.7529>,<at 0.708 0.7084>,<at 0.75 0.5711>,<at 0.791 0.4735>,<at 0.833 0.3231>,<at 0.875 0.4233>,<at 0.916 0.8388>,<at 0.958 0.9047>,<at 1 0.9216>]}
         GreenCurve {keys [<at 0 0.6308>,<at 0.041 0.5711>,<at 0.083 0.5333>,<at 0.125 0.5583>,<at 0.166 0.5333>,<at 0.208 0.5271>,<at 0.25 0.6376>,<at 0.291 0.7529>,<at 0.333 0.7835>,<at 0.375 0.8148>,<at 0.416 0.7991>,<at 0.458 0.8148>,<at 0.5 0.8469>,<at 0.541 0.7991>,<at 0.583 0.7991>,<at 0.625 0.8148>,<at 0.666 0.7991>,<at 0.708 0.7529>,<at 0.75 0.5711>,<at 0.791 0.5029>,<at 0.833 0.5029>,<at 0.875 0.6172>,<at 0.916 0.5333>,<at 0.958 0.5711>,<at 1 0.6308>]}
         RedCurve {keys [<at 0 0.305>,<at 0.041 0.2623>,<at 0.083 0.2664>,<at 0.125 0.2831>,<at 0.166 0.7084>,<at 0.208 0.6867>,<at 0.25 0.8148>,<at 0.291 0.8879>,<at 0.333 0.8963>,<at 0.375 0.9131>,<at 0.416 0.8714>,<at 0.458 0.8714>,<at 0.5 0.9131>,<at 0.541 0.8632>,<at 0.583 0.8714>,<at 0.625 0.9131>,<at 0.666 0.8963>,<at 0.708 0.8714>,<at 0.75 0.7529>,<at 0.791 0.7529>,<at 0.833 0.8469>,<at 0.875 0.8632>,<at 0.916 0.1878>,<at 0.958 0.1746>,<at 1 0.305>]}
        }}
      )
      B=&MakeFloat4FromVectorAndFloat("Vector (RGB)"=<1 1 1>, "Float (Alpha)"=$System.FadeInOpacity)
     )
    )) {isSetParameters=true}
   &ScaleColor(
    Scale Mode=::NewEnumerator1
    Scale RGBA=&Multiply_Vector4(
     A=&Vector4FromCurve(
      CurveIndex=Linked() {linkedVariable {name=NPC.FortniteNPC.TimeOfDay_0-1, type=NiagaraFloat}}
      Vector4Curve=DataInterface() {propertyValues {
        AlphaCurve {keys [<at 0 1>,<at 0.125 1>,<at 0.25 1>,<at 0.375 1>,<at 0.5 1>,<at 0.625 1>,<at 0.75 1>,<at 0.875 1>,<at 1 1>]}
        BlueCurve {keys [<at 0 0.9216>,<at 0.041 0.8963>,<at 0.083 0.8308>,<at 0.125 0.7835>,<at 0.166 0.6172>,<at 0.208 0.5395>,<at 0.25 0.6172>,<at 0.291 0.6939>,<at 0.333 0.7379>,<at 0.375 0.7835>,<at 0.416 0.7991>,<at 0.458 0.8148>,<at 0.5 0.8308>,<at 0.541 0.8148>,<at 0.583 0.7991>,<at 0.625 0.7991>,<at 0.666 0.7529>,<at 0.708 0.7084>,<at 0.75 0.5711>,<at 0.791 0.4735>,<at 0.833 0.3231>,<at 0.875 0.4233>,<at 0.916 0.8388>,<at 0.958 0.9047>,<at 1 0.9216>]}
        GreenCurve {keys [<at 0 0.6308>,<at 0.041 0.5711>,<at 0.083 0.5333>,<at 0.125 0.5583>,<at 0.166 0.5333>,<at 0.208 0.5271>,<at 0.25 0.6376>,<at 0.291 0.7529>,<at 0.333 0.7835>,<at 0.375 0.8148>,<at 0.416 0.7991>,<at 0.458 0.8148>,<at 0.5 0.8469>,<at 0.541 0.7991>,<at 0.583 0.7991>,<at 0.625 0.8148>,<at 0.666 0.7991>,<at 0.708 0.7529>,<at 0.75 0.5711>,<at 0.791 0.5029>,<at 0.833 0.5029>,<at 0.875 0.6172>,<at 0.916 0.5333>,<at 0.958 0.5711>,<at 1 0.6308>]}
        RedCurve {keys [<at 0 0.305>,<at 0.041 0.2623>,<at 0.083 0.2664>,<at 0.125 0.2831>,<at 0.166 0.7084>,<at 0.208 0.6867>,<at 0.25 0.8148>,<at 0.291 0.8879>,<at 0.333 0.8963>,<at 0.375 0.9131>,<at 0.416 0.8714>,<at 0.458 0.8714>,<at 0.5 0.9131>,<at 0.541 0.8632>,<at 0.583 0.8714>,<at 0.625 0.9131>,<at 0.666 0.8963>,<at 0.708 0.8714>,<at 0.75 0.7529>,<at 0.791 0.7529>,<at 0.833 0.8469>,<at 0.875 0.8632>,<at 0.916 0.1878>,<at 0.958 0.1746>,<at 1 0.305>]}
       }}
     )
     B=&MakeFloat4FromVectorAndFloat("Vector (RGB)"=<1 1 1>, "Float (Alpha)"=$System.FadeInOpacity)
    )
    ScaleRGBA=true
   )
  ]
  rendererClasses [&NiagaraMeshRendererProperties]
  renderers [
   {rendererClass=&NiagaraMeshRendererProperties, rendererIndex=0}
   {rendererClass=&NiagaraMeshRendererProperties, rendererIndex=1}
   {rendererClass=&NiagaraMeshRendererProperties, rendererIndex=2}
   {rendererClass=&NiagaraMeshRendererProperties, rendererIndex=3}
   {rendererClass=&NiagaraMeshRendererProperties, rendererIndex=4}
   {rendererClass=&NiagaraMeshRendererProperties, rendererIndex=5}
   {rendererClass=&NiagaraMeshRendererProperties, rendererIndex=6}
   {rendererClass=&NiagaraMeshRendererProperties, rendererIndex=7}
   {rendererClass=&NiagaraMeshRendererProperties, rendererIndex=8}
   {rendererClass=&NiagaraMeshRendererProperties, rendererIndex=9}
   {rendererClass=&NiagaraMeshRendererProperties, rendererIndex=10}
   {rendererClass=&NiagaraMeshRendererProperties, rendererIndex=11}
   {rendererClass=&NiagaraMeshRendererProperties, rendererIndex=12}
   {rendererClass=&NiagaraMeshRendererProperties, rendererIndex=13}
   {rendererClass=&NiagaraMeshRendererProperties, rendererIndex=14}
   {rendererClass=&NiagaraMeshRendererProperties, rendererIndex=15}
   {rendererClass=&NiagaraMeshRendererProperties, rendererIndex=16}
   {rendererClass=&NiagaraMeshRendererProperties, rendererIndex=17}
   {rendererClass=&NiagaraMeshRendererProperties, rendererIndex=18}
   {rendererClass=&NiagaraMeshRendererProperties, rendererIndex=19}
   {rendererClass=&NiagaraMeshRendererProperties, rendererIndex=20}
   {rendererClass=&NiagaraMeshRendererProperties, rendererIndex=21}
   {rendererClass=&NiagaraMeshRendererProperties, rendererIndex=22}
   {rendererClass=&NiagaraMeshRendererProperties, rendererIndex=23}
   {rendererClass=&NiagaraMeshRendererProperties, rendererIndex=24}
   {rendererClass=&NiagaraMeshRendererProperties, rendererIndex=25}
  ]
 }}
summary {
 systemName=NS_Backpack_PlumedStare_Idle
 emitters [{emitterName=FlamesSplinesCombined, rendererClasses [&NiagaraMeshRendererProperties]}]
 userVariables [
  User.bisADS {type=NiagaraBool}
  User.bIsADSOther {type=NiagaraBool}
  User.bisFootHidden {type=NiagaraBool}
  User.bIsFrontEnd {type=NiagaraBool}
  User.bIsFrontEndPreview {type=NiagaraBool}
  User.bIsHandHidden {type=NiagaraBool}
 ]
}
