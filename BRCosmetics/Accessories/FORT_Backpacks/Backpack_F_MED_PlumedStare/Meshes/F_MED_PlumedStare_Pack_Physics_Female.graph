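# Physics asset for the female PlumedStare backpack cosmetic, setting up a dangling three-segment rope that hangs off the pack.
# Capsule bodies on the rope bones form a chain from the pack, with narrow swing limits that loosen past the first joint and zero-degree twist limits.
# Larger pelvis and lower-spine capsules approximate the wearer's body, linked by free angular constraints with no constraint tying them to the pack.
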PhysicsAsset /BRCosmetics/Accessories/FORT_Backpacks/Backpack_F_MED_PlumedStare/Meshes/F_MED_PlumedStare_Pack_Physics_Female
bodies {
 pack [{shapeType=Capsule, length=1, radius=4, shapeName=pack_capsule, rotation=<rot 0 -90 0>}]
 pelvis [{shapeType=Capsule, length=10, radius=9.5, shapeName=pelvis_capsule, center=<1 0 0>, rotation=<rot 90 0 -2>}]
 rope_01_dyn [{shapeType=Capsule, length=3.0147, radius=1.4843, shapeName=rope_01_dyn_capsule, center=<2 0 0>, rotation=<rot 90 0 0>}]
 rope_02_dyn [{shapeType=Capsule, length=3.0199, radius=1.4792, shapeName=rope_02_dyn_capsule, center=<3 0 0>, rotation=<rot 90 0 0>}]
 rope_03_dyn [{shapeType=Capsule, length=2.1621, radius=3.6093, shapeName=rope_03_dyn_capsule, center=<5 0 -0.8>, rotation=<rot 87 0 0>}]
 spine_01 [{shapeType=Capsule, length=10, radius=9.5, shapeName=spine_01_capsule, center=<4 0 0>, rotation=<rot 90 0 0>}]
 spine_02 [{shapeType=Capsule, length=10, radius=10, shapeName=spine_02_capsule, center=<5 0 0>, rotation=<rot 90 0 0>}]
}
constraints [
 {bone1Name=rope_01_dyn, bone2Name=pack, swing1LimitDegrees=8, swing2LimitDegrees=14, twistLimitDegrees=0}
 {
  bone1Name=spine_02
  bone2Name=spine_01
  swing1LimitDegrees=45
  swing1Motion=Free
  swing2LimitDegrees=45
  swing2Motion=Free
  twistLimitDegrees=45
  twistMotion=Free
 }
 {bone1Name=rope_02_dyn, bone2Name=rope_01_dyn, swing1LimitDegrees=22, swing2LimitDegrees=28, twistLimitDegrees=0}
 {bone1Name=rope_03_dyn, bone2Name=rope_02_dyn, swing1LimitDegrees=25, swing2LimitDegrees=28, twistLimitDegrees=0}
 {
  bone1Name=spine_01
  bone2Name=pelvis
  swing1LimitDegrees=45
  swing1Motion=Free
  swing2LimitDegrees=45
  swing2Motion=Free
  twistLimitDegrees=45
  twistMotion=Free
 }
]
