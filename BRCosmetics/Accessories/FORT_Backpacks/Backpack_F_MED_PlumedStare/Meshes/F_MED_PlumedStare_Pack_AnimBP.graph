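# Animation Blueprint for the PlumedStare backpack, adding secondary motion to its pack, feathers and wings.
# A cached body-joint pose feeds male or female RigidBody simulations, blended by isMale, then an AnimDynamics chain for the feathers.
# Spine yaw shifts and rolls the pack, while a time-based sine bobs both wings.
# The event graph also samples body joints, gravity override and frontend-dependent sim settings; the wing AnimDynamics nodes are disconnected from the output.
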AnimBlueprint /BRCosmetics/Accessories/FORT_Backpacks/Backpack_F_MED_PlumedStare/Meshes/F_MED_PlumedStare_Pack_AnimBP
target_skeleton=&/F_MED_PlumedStare_Pack_Skeleton
graphs [
 AnimGraph {nodes [
   n1 = Anim Dynamics - Bone: wing_l_dyn!(ComponentPose, Alpha=1, GravityOverride:Vector="0,0,-300", ExternalForce="0,0,0") -> (Pose) {
    angularDampingOverride=1
    angularSpringConstant=10000
    bLinearSpring=true
    bOverrideLinearDamping=true
    bUseGravityOverride=true
    gravityScale=0.35
    linearDampingOverride=1
    linearSpringConstant=0.5
    simulationSpace=RootRelative
    alphaBoolBlend {blendOption=Linear}
    boundBone {boneName=wing_l_dyn}
    componentAppliedLinearAccClamp=<3000 3000 3000>
    componentLinearAccScale=<0.18 0.18 0.18>
    componentLinearVelScale=<0.18 0.18 0.18>
    gravityOverride {z=-980}
    physicsBodyDefinitions [{
      boundBone {boneName=wing_l_dyn}
      boxExtents=<7.75 7.75 7.75>
      constraintSetup {angularLimitsMax=<0 0 0>, angularLimitsMin=<0 0 0>, angularTarget=<1 0 0>, linearAxesMax=<1 0.25 0.5>, linearAxesMin=<-2.3 0 -0.5>}
      localJointOffset=<1 0 0>
     }]
   }
   Anim Dynamics - Bone: wing_r_dyn!(ComponentPose=@n1, Alpha=1, GravityOverride:Vector="0,0,-300", ExternalForce="0,0,0") -> (Pose) {
    angularDampingOverride=1
    angularSpringConstant=10000
    bLinearSpring=true
    bOverrideLinearDamping=true
    bUseGravityOverride=true
    gravityScale=0.35
    linearDampingOverride=1
    linearSpringConstant=0.5
    simulationSpace=RootRelative
    alphaBoolBlend {blendOption=Linear}
    boundBone {boneName=wing_r_dyn}
    componentAppliedLinearAccClamp=<3000 3000 3000>
    componentLinearAccScale=<0.18 0.18 0.18>
    componentLinearVelScale=<0.17 0.17 0.17>
    gravityOverride {z=-980}
    physicsBodyDefinitions [{
      boundBone {boneName=wing_r_dyn}
      boxExtents=<7.75 7.75 7.75>
      constraintSetup {angularLimitsMax=<0 0 0>, angularLimitsMin=<0 0 0>, angularTarget=<1 0 0>, linearAxesMax=<1 0.25 0.5>, linearAxesMin=<-2.3 0 -0.5>}
      localJointOffset=<1 0 0>
     }]
   }
   n2 = "Transform (Modify) Bone - Bone: pack"(
    ComponentPose=Local To Component(LocalPose=Use cached pose 'AnimDynamics Cache'())
    Translation=Get backpack_translation()
    Rotation=Get backpack_rotate()
   ) {rotationMode=BMM_Additive, translationMode=BMM_Additive, alphaBoolBlend {blendOption=Linear}, boneToModify {boneName=pack}}
   n3 = "Transform (Modify) Bone - Bone: rope_01_base"!(ComponentPose=@n2, bAlphaBoolEnabled:bool=True, Translation="0,0,0", Rotation="0,0,-10", Scale="1,1,1") -> (Pose) {
    alphaInputType=Bool
    rotationMode=BMM_Additive
    alphaBoolBlend {blendOption=Linear}
    boneToModify {boneName=rope_01_base}
    rotation {roll=-10}
   }
   n4 = "Transform (Modify) Bone - Bone: wing_r_dyn"(ComponentPose=@n3, Alpha=1, Translation="-0.15,0,0", Rotation="-6.5,0,0", Scale="1,1,1") {
    rotationMode=BMM_Additive
    translationMode=BMM_Additive
    alphaBoolBlend {blendOption=Linear}
    boneToModify {boneName=wing_r_dyn}
    rotation {pitch=-6.5}
    translation {x=-0.15}
   }
   n5 = Get Float() -> (Float:Vector)
   n6 = "Transform (Modify) Bone - Bone: wing_r_dyn"(ComponentPose=@n4, Alpha=1, Translation=@n5, Rotation="0,0,0", Scale="1,1,1") {rotationMode=BMM_Additive, translationMode=BMM_Additive, alphaBoolBlend {blendOption=Linear}, boneToModify {boneName=wing_r_dyn}}
   n7 = "Transform (Modify) Bone - Bone: wing_l_dyn"(ComponentPose=@n6, Alpha=1, Translation=@n5, Rotation="0,0,0", Scale="1,1,1") {rotationMode=BMM_Additive, translationMode=BMM_Additive, alphaBoolBlend {blendOption=Linear}, boneToModify {boneName=wing_l_dyn}}
   Output Pose(Result=Component To Local(ComponentPose=@n7)) {layerGroup=None}
   n8 = "Transform (Modify) Bone - Bone: pack"!(ComponentPose=Local To Component(LocalPose=Use cached pose 'Body joints position'()), bAlphaBoolEnabled:bool, Translation="0,0.5,0") -> (Pose) {
    alphaInputType=Bool
    translationMode=BMM_Additive
    translationSpace=BCS_BoneSpace
    alphaBoolBlend {blendOption=Linear}
    boneToModify {boneName=pack}
    translation {y=0.5}
   }
   n9 = RigidBody!(
    ComponentPose=@n8
    bAlphaBoolEnabled:bool
    OverrideWorldGravity:Vector="0,0,0"
    ExternalForce
    SimSpaceSettings:Sim Space Settings Structure=Get Sim Space Settings() -> (Sim Space Settings:Sim Space Settings Structure)
   ) -> (Pose) {
    alphaInputType=Bool
    bOverrideWorldGravity=true
    overridePhysicsAsset=&/F_MED_PlumedStare_Pack_Physics_Female
    simulationSpace=BaseBoneSpace
    alphaBoolBlend {blendOption=Linear}
    baseBoneRef {boneName=root}
    componentAppliedLinearAccClamp=<3000 3000 3000>
    componentLinearAccScale=<0.28 0.28 0.28>
    componentLinearVelScale=<0.28 0.28 0.28>
    simSpaceSettings {
     maxAngularAcceleration=1000
     maxAngularVelocity=3
     maxLinearAcceleration=700
     maxLinearVelocity=500
     velocityScaleZ=0
     worldAlpha=0.3
     externalLinearDragV=<0.2 0.2>
    }
   }
   n10 = Component To Local(ComponentPose=@n9)
   n11 = "Transform (Modify) Bone - Bone: pack"(ComponentPose=Local To Component(LocalPose=Use cached pose 'Body joints position'()), Translation="0,1,0") {
    alphaInputType=Bool
    translationMode=BMM_Additive
    translationSpace=BCS_BoneSpace
    alphaBoolBlend {blendOption=Linear}
    boneToModify {boneName=pack}
    translation {y=1}
   }
   n12 = RigidBody(ComponentPose=@n11, OverrideWorldGravity="0,0,0", SimSpaceSettings=Get Sim Space Settings()) {
    alphaInputType=Bool
    bOverrideWorldGravity=true
    overridePhysicsAsset=&/F_MED_PlumedStare_Pack_Physics
    simulationSpace=BaseBoneSpace
    alphaBoolBlend {blendOption=Linear}
    baseBoneRef {boneName=root}
    componentAppliedLinearAccClamp=<3000 3000 3000>
    componentLinearAccScale=<0.28 0.28 0.28>
    componentLinearVelScale=<0.28 0.28 0.28>
    simSpaceSettings {
     maxAngularAcceleration=1000
     maxAngularVelocity=3
     maxLinearAcceleration=700
     maxLinearVelocity=500
     velocityScaleZ=0
     worldAlpha=0.3
     externalLinearDragV=<0.2 0.2>
    }
   }
   n13 = Component To Local(ComponentPose=@n12)
   n14 = Get isMale() -> (isMale:bool)
   n15 = Anim Dynamics - Bone: featherA_dyn!(
    ComponentPose=Local To Component(LocalPose=Blend Poses by bool(BlendPose_0=@n13, BlendPose_1=@n10, bActiveValue=@n14) {blendType=Linear})
    Alpha=1
    GravityOverride:Vector="0,0,-980"
    ExternalForce="0,0,0"
   ) -> (Pose) {
    angularDampingOverride=1
    angularSpringConstant=10000
    bAngularSpring=true
    bOverrideAngularDamping=true
    bUseGravityOverride=true
    gravityScale=0.35
    alphaBoolBlend {blendOption=Linear}
    boundBone {boneName=featherA_dyn}
    componentAppliedLinearAccClamp=<3000 3000 3000>
    componentLinearAccScale=<0.4 0.4 0.5>
    componentLinearVelScale=<0.4 0.4 0.5>
    gravityOverride {z=-980}
    physicsBodyDefinitions [{
      boundBone {boneName=featherA_dyn}
      boxExtents=<15 15 15>
      constraintSetup {angularLimitsMax=<0 25 25>, angularLimitsMin=<0 -25 -25>, angularTarget=<1 0 0>}
      localJointOffset=<4 0 0>
     }]
   }
   n16 = Anim Dynamics - Bone: featherB_01_dyn!(ComponentPose=@n15, Alpha=1, GravityOverride:Vector="0,0,-980", ExternalForce="0,0,0") -> (Pose) {
    angularDampingOverride=1
    angularSpringConstant=10000
    bAngularSpring=true
    bOverrideAngularDamping=true
    bUseGravityOverride=true
    gravityScale=0.35
    alphaBoolBlend {blendOption=Linear}
    boundBone {boneName=featherB_01_dyn}
    componentAppliedLinearAccClamp=<3000 3000 3000>
    componentLinearAccScale=<0.4 0.4 0.5>
    componentLinearVelScale=<0.4 0.4 0.5>
    gravityOverride {z=-980}
    physicsBodyDefinitions [{
      boundBone {boneName=featherB_01_dyn}
      boxExtents=<15 15 15>
      constraintSetup {angularLimitsMax=<0 25 25>, angularLimitsMin=<0 -25 -25>, angularTarget=<1 0 0>}
      localJointOffset=<4 0 0>
     }]
   }
   n17 = Anim Dynamics - Bone: featherB_02_dyn!(ComponentPose=@n16, Alpha=1, GravityOverride:Vector="0,0,-980", ExternalForce="0,0,0") -> (Pose) {
    angularDampingOverride=1
    angularSpringConstant=10000
    bAngularSpring=true
    bOverrideAngularDamping=true
    bUseGravityOverride=true
    gravityScale=0.35
    alphaBoolBlend {blendOption=Linear}
    boundBone {boneName=featherB_02_dyn}
    componentAppliedLinearAccClamp=<3000 3000 3000>
    componentLinearAccScale=<0.4 0.4 0.5>
    componentLinearVelScale=<0.4 0.4 0.5>
    gravityOverride {z=-980}
    physicsBodyDefinitions [{
      boundBone {boneName=featherB_02_dyn}
      boxExtents=<15 15 15>
      constraintSetup {angularLimitsMax=<0 25 25>, angularLimitsMin=<0 -25 -25>, angularTarget=<1 0 0>}
      localJointOffset=<4 0 0>
     }]
   }
   n18 = Anim Dynamics - Bone: featherC_dyn!(ComponentPose=@n17, Alpha=1, GravityOverride:Vector="0,0,-980", ExternalForce="0,0,0") -> (Pose) {
    angularDampingOverride=1
    angularSpringConstant=10000
    bAngularSpring=true
    bOverrideAngularDamping=true
    bUseGravityOverride=true
    gravityScale=0.35
    alphaBoolBlend {blendOption=Linear}
    boundBone {boneName=featherC_dyn}
    componentAppliedLinearAccClamp=<3000 3000 3000>
    componentLinearAccScale=<0.4 0.4 0.5>
    componentLinearVelScale=<0.4 0.4 0.5>
    gravityOverride {z=-980}
    physicsBodyDefinitions [{
      boundBone {boneName=featherC_dyn}
      boxExtents=<15 15 15>
      constraintSetup {angularLimitsMax=<0 25 25>, angularLimitsMin=<0 -25 -25>, angularTarget=<1 0 0>}
      localJointOffset=<4 0 0>
     }]
   }
   n19 = Anim Dynamics - Bone: featherD_01_dyn!(ComponentPose=@n18, Alpha=1, GravityOverride:Vector="0,0,-980", ExternalForce="0,0,0") -> (Pose) {
    angularDampingOverride=1
    angularSpringConstant=10000
    bAngularSpring=true
    bOverrideAngularDamping=true
    bUseGravityOverride=true
    gravityScale=0.35
    alphaBoolBlend {blendOption=Linear}
    boundBone {boneName=featherD_01_dyn}
    componentAppliedLinearAccClamp=<3000 3000 3000>
    componentLinearAccScale=<0.4 0.4 0.5>
    componentLinearVelScale=<0.4 0.4 0.5>
    gravityOverride {z=-980}
    physicsBodyDefinitions [{
      boundBone {boneName=featherD_01_dyn}
      boxExtents=<15 15 15>
      constraintSetup {angularLimitsMax=<0 25 25>, angularLimitsMin=<0 -25 -25>, angularTarget=<1 0 0>}
      localJointOffset=<4 0 0>
     }]
   }
   n20 = Anim Dynamics - Bone: featherD_02_dyn!(ComponentPose=@n19, Alpha=1, GravityOverride:Vector="0,0,-980", ExternalForce="0,0,0") -> (Pose) {
    angularDampingOverride=1
    angularSpringConstant=10000
    bAngularSpring=true
    bOverrideAngularDamping=true
    bUseGravityOverride=true
    gravityScale=0.35
    alphaBoolBlend {blendOption=Linear}
    boundBone {boneName=featherD_02_dyn}
    componentAppliedLinearAccClamp=<3000 3000 3000>
    componentLinearAccScale=<0.4 0.4 0.5>
    componentLinearVelScale=<0.4 0.4 0.5>
    gravityOverride {z=-980}
    physicsBodyDefinitions [{
      boundBone {boneName=featherD_02_dyn}
      boxExtents=<15 15 15>
      constraintSetup {angularLimitsMax=<0 25 25>, angularLimitsMin=<0 -25 -25>, angularTarget=<1 0 0>}
      localJointOffset=<4 0 0>
     }]
   }
   n21 = Anim Dynamics - Bone: featherE_dyn!(ComponentPose=@n20, Alpha=1, GravityOverride:Vector="0,0,-980", ExternalForce="0,0,0") -> (Pose) {
    angularDampingOverride=1
    angularSpringConstant=10000
    bAngularSpring=true
    bOverrideAngularDamping=true
    bUseGravityOverride=true
    gravityScale=0.35
    alphaBoolBlend {blendOption=Linear}
    boundBone {boneName=featherE_dyn}
    componentAppliedLinearAccClamp=<3000 3000 3000>
    componentLinearAccScale=<0.4 0.4 0.5>
    componentLinearVelScale=<0.4 0.4 0.5>
    gravityOverride {z=-980}
    physicsBodyDefinitions [{
      boundBone {boneName=featherE_dyn}
      boxExtents=<15 15 15>
      constraintSetup {angularLimitsMax=<0 15 15>, angularLimitsMin=<0 -15 -15>, angularTarget=<1 0 0>}
      localJointOffset=<4 0 0>
     }]
   }
   Save cached pose 'AnimDynamics Cache'(Pose=Component To Local(ComponentPose=@n21)) {cacheName=AnimDynamics Cache}
   n22 = "Transform (Modify) Bone - Bone: body_root"(ComponentPose=Local To Component(LocalPose=Input Pose()), Alpha=1, Translation="0,0,0", Rotation="0,0,0", Scale="1,1,1") {
    rotationMode=BMM_Replace
    rotationSpace=BCS_WorldSpace
    scaleSpace=BCS_BoneSpace
    translationMode=BMM_Replace
    translationSpace=BCS_WorldSpace
    alphaBoolBlend {blendOption=Linear}
    alphaScaleBiasClamp {bInterpResult=true, interpSpeedDecreasing=1, interpSpeedIncreasing=1, inRange {max=0, min=-40}, outRange {max=0, min=1}}
    boneToModify {boneName=body_root}
   }
   n23 = "Transform (Modify) Bone - Bone: pelvis"(ComponentPose=@n22, Alpha=1, Translation="0,0,0", Rotation="0,0,0", Scale="1,1,1") {
    rotationMode=BMM_Replace
    rotationSpace=BCS_WorldSpace
    scaleSpace=BCS_BoneSpace
    translationMode=BMM_Replace
    translationSpace=BCS_WorldSpace
    alphaBoolBlend {blendOption=Linear}
    alphaScaleBiasClamp {inRange {max=0, min=-40}, outRange {max=0, min=1}}
    boneToModify {boneName=pelvis}
   }
   n24 = "Transform (Modify) Bone - Bone: spine_01"(ComponentPose=@n23, Alpha=1, Translation="0,0,0", Rotation="0,0,0", Scale="1,1,1") {
    rotationMode=BMM_Replace
    rotationSpace=BCS_WorldSpace
    scaleSpace=BCS_BoneSpace
    translationMode=BMM_Replace
    translationSpace=BCS_WorldSpace
    alphaBoolBlend {blendOption=Linear}
    alphaScaleBiasClamp {inRange {max=0, min=-40}, outRange {max=0, min=1}}
    boneToModify {boneName=spine_01}
   }
   n25 = "Transform (Modify) Bone - Bone: spine_02"(ComponentPose=@n24, Alpha=1, Translation="0,0,0", Rotation="0,0,0", Scale="1,1,1") {
    rotationMode=BMM_Replace
    rotationSpace=BCS_WorldSpace
    scaleSpace=BCS_BoneSpace
    translationMode=BMM_Replace
    translationSpace=BCS_WorldSpace
    alphaBoolBlend {blendOption=Linear}
    alphaScaleBiasClamp {inRange {max=0, min=-40}, outRange {max=0, min=1}}
    boneToModify {boneName=spine_02}
   }
   n26 = "Transform (Modify) Bone - Bone: spine_03"(ComponentPose=@n25, Alpha=1, Translation="0,0,0", Rotation="0,0,0", Scale="1,1,1") {
    rotationMode=BMM_Replace
    rotationSpace=BCS_WorldSpace
    scaleSpace=BCS_BoneSpace
    translationMode=BMM_Replace
    translationSpace=BCS_WorldSpace
    alphaBoolBlend {blendOption=Linear}
    alphaScaleBiasClamp {inRange {max=0, min=-40}, outRange {max=0, min=1}}
    boneToModify {boneName=spine_03}
   }
   n27 = "Transform (Modify) Bone - Bone: spine_04"(ComponentPose=@n26, Alpha=1, Translation="0,0,0", Rotation="0,0,0", Scale="1,1,1") {
    rotationMode=BMM_Replace
    rotationSpace=BCS_WorldSpace
    scaleSpace=BCS_BoneSpace
    translationMode=BMM_Replace
    translationSpace=BCS_WorldSpace
    alphaBoolBlend {blendOption=Linear}
    alphaScaleBiasClamp {inRange {max=0, min=-40}, outRange {max=0, min=1}}
    boneToModify {boneName=spine_04}
   }
   Save cached pose 'Body joints position'(Pose=Component To Local(ComponentPose=@n27)) {cacheName=Body joints position}
  ]}
 EventGraph {nodes [
   Event BlueprintUpdateAnimation()
   > n1 = Sequence!(execute) -> (then_0, then_1, then_2:Exec, then_3:Exec, then_4:Exec)
   n2 = GetPartSkeletalMeshComponent()
   Is Valid(exec=@n1.then_2, InputObject=@n2)
   n3 = GetAttachParent!(self:Scene Component Object Reference=@n2) -> (ReturnValue:Scene Component Object Reference)
   > Is Valid(InputObject=@n3)
   > Set Body_Mesh(Body_Mesh:Scene Component Object Reference=@n3) -> (Output_Get:Scene Component Object Reference)
   > n4 = Sequence!(execute) -> (then_0, then_1, then_2, then_3, then_4, then_5:Exec, then_6:Exec)
   n5 = Get Body_Mesh() -> (Body_Mesh:Scene Component Object Reference)
   n6 = BreakTransform(InTransform=GetSocketTransform(self=@n5, InSocketName=spine_01, TransformSpace=RTS_World))
   Set Spine1_location(execute=@n4.then_1, Spine1_location:Vector=@n6) -> (Output_Get:Vector)
   > Set Spine1_rotation(Spine1_rotation:Rotator=@n6.Rotation) -> (Output_Get:Rotator)
   n7 = BreakTransform(InTransform=GetSocketTransform(self=@n5, InSocketName=spine_02, TransformSpace=RTS_World))
   Set Spine2_location(execute=@n4.then_2, Spine2_location:Vector=@n7) -> (Output_Get:Vector)
   > Set Spine2_rotation(Spine2_rotation:Rotator=@n7.Rotation) -> (Output_Get:Rotator)
   n8 = BreakTransform(InTransform=GetSocketTransform(self=@n5, InSocketName=root, TransformSpace=RTS_World))
   Set root_location(execute=@n4.then_6, root_location:Vector=@n8) -> (Output_Get:Vector)
   > Set root_space_rotation(root_space_rotation:Rotator=@n8.Rotation) -> (Output_Get:Rotator)
   n9 = BreakTransform(InTransform=GetSocketTransform(self=@n5, InSocketName=pelvis, TransformSpace=RTS_World))
   Set pelvis_location(execute=@n4.then_0, pelvis_location:Vector=@n9) -> (Output_Get:Vector)
   > Set pelvis_rotation(pelvis_rotation:Rotator=@n9.Rotation) -> (Output_Get:Rotator)
   n10 = BreakTransform(InTransform=GetSocketTransform(self=@n5, InSocketName=spine_04, TransformSpace=RTS_World))
   Set Spine4_location(execute=@n4.then_4, Spine4_location:Vector=@n10) -> (Output_Get:Vector)
   > Set Spine4_rotation(Spine4_rotation:Rotator=@n10.Rotation) -> (Output_Get:Rotator)
   n11 = BreakTransform(InTransform=GetSocketTransform(self=@n5, InSocketName=spine_03, TransformSpace=RTS_World))
   Set Spine3_location(execute=@n4.then_3, Spine3_location:Vector=@n11) -> (Output_Get:Vector)
   > Set Spine3_rotation(Spine3_rotation:Rotator=@n11.Rotation) -> (Output_Get:Rotator)
   n12 = BreakTransform(InTransform=GetSocketTransform(self=@n5, InSocketName=root, TransformSpace=RTS_World))
   Set body_root_location(execute=@n4.then_5, body_root_location:Vector=@n12) -> (Output_Get:Vector)
   > Set body_root_rotation(body_root_rotation:Rotator=@n12.Rotation) -> (Output_Get:Rotator)
   n13 = Branch(execute=@n1.then_4, Condition=Get bIsInFrontEnd())
   > Set Sim Space Settings(Sim Space Settings:Sim Space Settings Structure=Get Sim Space Settings_FE() -> (Sim Space Settings_FE:Sim Space Settings Structure)) -> (Output_Get:Sim Space Settings Structure)
   Set Sim Space Settings(execute=@n13.else, Sim Space Settings=Get Sim Space Settings_InZone() -> (Sim Space Settings_InZone:Sim Space Settings Structure))
   n14 = Gravity Override State Function DEPRECATED(
    execute=@n1.then_0
    CharacterPartAnimInstance=Self-Reference()
    GameTimeInSeconds=GetGameTimeInSeconds()
    GravityOverrideParams=Get Gravity Override Params()
   )
   > Set GravityOverride(GravityOverride=@n14)
   n15 = GetOwnerSkeletalMeshComponent()
   n16 = BreakRotator(InRot=BreakTransform(InTransform=GetDeltaTransformFromRefPose(self=@n15, BoneName=spine_05, BaseName=spine_01)).Rotation)
   Set backpack_rotate(execute=@n1.then_1, backpack_rotate=MakeRotator(Roll=MapRangeClamped(Value=@n16.Yaw, InRangeA=0, InRangeB=-20, OutRangeB=25)))
   > n17 = Set backpack_translation(backpack_translation=MakeVector(Y=MapRangeClamped(Value=@n16.Yaw, InRangeA=0, InRangeB=-20, OutRangeB=-2)))
   n18 = Event BlueprintProcessAnimInputEvent()
   n19 = Branch(execute=@n17.then, Condition=OR Boolean(A=IsNamed(self=@n18.AnimInputEvent, InEventName=aie.Player.Zipline.Entered), B=Get bIsOnGrindRail()))
   > Set OnZipline(OnZipline:bool=1) -> (Output_Get:bool)
   > Set ZiplineAlpha(ZiplineAlpha:double=0) -> (Output_Get:double)
   Branch(
    execute=@n19.else
    Condition=OR Boolean!(A=IsNamed(self=@n18.AnimInputEvent, InEventName=aie.Player.Zipline.Exited), B=Get bIsMoving2D(), C:bool=Get bIsJumping()) -> (ReturnValue)
   )
   > Set OnZipline(OnZipline=false)
   > Set ZiplineAlpha(ZiplineAlpha=1)
   n20 = "Sin (Radians)"(A=float * float(A=float * float(A=1.1, B=1.2), B=GetRealTimeSeconds()))
   Set Float(execute=@n1.then_3, Float:Vector=MakeVector(Y=0, Z=float * float(A=@n20, B=1))) -> (Output_Get:Vector)
   Event OnCharacterCustomizationLoaded()
   > n21 = Set Owning Pawn(Owning Pawn:Pawn Object Reference=TryGetFortPawn()) -> (Output_Get:Pawn Object Reference)
   > Is Valid(InputObject=@n21)
   > n22 = Sequence!(execute) -> (then_0, then_1, then_2)
   n23 = FindGender!(execute=@n22.then_1, Object:Object Reference=Get Owning Pawn() -> (Owning Pawn:Pawn Object Reference)) -> (True:Exec, False:Exec, OutGender:EFortCustomGender Enum=Invalid)
   > Set isMale(isMale:bool="Equal (Enum)"(A:EFortCustomGender Enum=@n23, B:EFortCustomGender Enum=Male)) -> (Output_Get:bool)
   > Set isFemale(isFemale:bool="Equal (Enum)"(A=@n23, B=Female)) -> (Output_Get:bool)
   n24 = FindBodyType!(execute=@n22.then_2, Object:Object Reference=Get Owning Pawn()) -> (True:Exec, False:Exec, OutBodyType:EFortCustomBodyType Enum=NONE)
   > Set isLarge(isLarge:bool="Equal (Enum)"(A:EFortCustomBodyType Enum=@n24, B:EFortCustomBodyType Enum=Large)) -> (Output_Get:bool)
   n25 = Cast To FortPlayerPawn(execute=@n22.then_0, Object=Get Owning Pawn())
   > n26 = Set Owning Player Pawn(Owning Player Pawn:Fort Player Pawn Object Reference=@n25) -> (Output_Get:Fort Player Pawn Object Reference)
   > Set Body Mesh(Body Mesh:Skeletal Mesh Component Object Reference=GetSkeletalMeshForPartType(self=@n26)) -> (Output_Get:Skeletal Mesh Component Object Reference)
   TryGetPawnOwner()
  ]}
]
variables [
 backpack_rotate {type=Rotator}
 backpack_translation {type=Vector}
 Body Mesh {type=SkeletalMeshComponent}
 Body_Mesh {type=SceneComponent, default=null}
 body_root_location {type=Vector}
 body_root_rotation {type=Rotator}
 Float {type=Vector}
 Gravity Override Params {type=GravityOverrideParamsStruct}
 GravityOverride {type=Vector}
 is launching {type=boolean}
 is launching_0 {type=number}
 isFemale {type=boolean}
 isLarge {type=boolean}
 isMale {type=boolean}
 NewVar {type=boolean}
 OnZipline {type=boolean}
 Owning Pawn {type=Pawn, default=null}
 Owning Player Pawn {type=FortPlayerPawn}
 pelvis_location {type=Vector}
 pelvis_rotation {type=Rotator}
 root_location {type=Vector}
 root_space_rotation {type=Rotator}
 sideWing left {type=Vector}
 sideWing right {type=Vector}
 Sim Space Settings {
  type=SimSpaceSettings
  default {
   dampingAlpha=1
   maxAngularAcceleration=1000
   maxAngularVelocity=3
   maxLinearAcceleration=700
   maxLinearVelocity=500
   velocityScaleZ=0
   worldAlpha=0.3
   externalLinearDragV=<0.2 0.2 0>
  }
 }
 Sim Space Settings_FE {
  type=SimSpaceSettings
  default {
   dampingAlpha=1
   maxAngularAcceleration=1000
   maxAngularVelocity=3
   maxLinearAcceleration=700
   maxLinearVelocity=500
   velocityScaleZ=0
   worldAlpha=0.3
   externalLinearDragV=<0.2 0.2 0>
  }
 }
 Sim Space Settings_InZone {
  type=SimSpaceSettings
  default {
   dampingAlpha=1
   maxAngularAcceleration=1000
   maxAngularVelocity=3
   maxLinearAcceleration=700
   maxLinearVelocity=500
   velocityScaleZ=0
   worldAlpha=0.3
   externalLinearDragV=<0.2 0.2 0>
  }
 }
 Spine1_location {type=Vector}
 Spine1_rotation {type=Rotator}
 Spine2_location {type=Vector}
 Spine2_rotation {type=Rotator}
 Spine3_location {type=Vector}
 Spine3_rotation {type=Rotator}
 Spine4_location {type=Vector}
 Spine4_rotation {type=Rotator}
 ZiplineAlpha {type=number}
]
